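# Post-process AnimBP for the cat pet on the shared dog skeleton, adding secondary head/neck motion and correcting leg helper bones.
# When bPetWantsAnimDynamics is set, neck_02 and head run root-relative AnimDynamics instead of passing the cached input through.
# Emote/DBNO selects a full-alpha angular-spring chain; otherwise pawn speed scales a looser chain.
# Rotation constraints then drive thigh, calf, heel and foot _OL bones from each joint and its parent.
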AnimBlueprint /BRCosmetics/Characters/Pets/Cat/Meshes/Cat_PostProcess_Skeleton_AnimBP
target_skeleton=/Game/Characters/Pets/Dog_Shared/Dog_Skeleton
graphs [
 AnimGraph {nodes [
   n1 = Get PawnSpeedForAnimDynamics() -> (PawnSpeedForAnimDynamics:float)
   n2 = Anim Dynamics - Bone: neck_02(ComponentPose=Local To Component(LocalPose=Use cached pose 'Sub Graph Cache'()), Alpha=@n1, ExternalForce="0,0,0") {
    angularDampingOverride=0.9
    angularSpringConstant=200
    bGravityOverrideInSimSpace=true
    bOverrideAngularDamping=true
    gravityScale=0.32
    lODThreshold=1
    numSolverIterationsPostUpdate=4
    numSolverIterationsPreUpdate=16
    simulationSpace=RootRelative
    alphaBoolBlend {blendOption=Linear}
    alphaScaleBiasClamp {bInterpResult=true, bMapRange=true, interpSpeedDecreasing=5, interpSpeedIncreasing=5, inRange {max=500}}
    boundBone {boneName=neck_02}
    componentAppliedLinearAccClamp=<100 100 100>
    componentLinearAccScale=<0.3 0.3 0.3>
    componentLinearVelScale=<0.3 0.3 0.3>
    physicsBodyDefinitions [{
      sphereCollisionRadius=0
      boundBone {boneName=neck_02}
      boxExtents=<3 10 10>
      constraintSetup {angularTargetAxis=AxisY, twistAxis=AxisY, angularLimitsMax=<30 20 20>, angularLimitsMin=<-30 -20 -20>, angularTarget=<0 0 0>}
      localJointOffset=<0 1 0>
     }]
    relativeSpaceBone {boneName=root}
   }
   n3 = Anim Dynamics - Bone: head(ComponentPose=@n2, Alpha=@n1, ExternalForce="0,0,0") {
    angularDampingOverride=0.9
    angularSpringConstant=200
    bGravityOverrideInSimSpace=true
    bOverrideAngularDamping=true
    gravityScale=0.32
    lODThreshold=1
    numSolverIterationsPostUpdate=4
    numSolverIterationsPreUpdate=16
    simulationSpace=RootRelative
    alphaBoolBlend {blendOption=Linear}
    alphaScaleBiasClamp {bInterpResult=true, bMapRange=true, interpSpeedDecreasing=5, interpSpeedIncreasing=5, inRange {max=500}}
    boundBone {boneName=head}
    componentAppliedLinearAccClamp=<100 100 100>
    componentLinearAccScale=<0.3 0.3 0.3>
    componentLinearVelScale=<0.3 0.3 0.3>
    physicsBodyDefinitions [{
      sphereCollisionRadius=0
      boundBone {boneName=head}
      boxExtents=<3 10 10>
      constraintSetup {angularTargetAxis=AxisY, twistAxis=AxisY, angularLimitsMax=<10 40 10>, angularLimitsMin=<-10 -40 -10>, angularTarget=<0 0 0>}
      localJointOffset=<0 1 0>
     }]
    relativeSpaceBone {boneName=root}
   }
   n4 = Anim Dynamics - Bone: neck_02(ComponentPose=Local To Component(LocalPose=Use cached pose 'Sub Graph Cache'()), Alpha=1, ExternalForce="0,0,0") {
    angularDampingOverride=0.99
    angularSpringConstant=1500
    bAngularSpring=true
    bGravityOverrideInSimSpace=true
    bOverrideAngularDamping=true
    gravityScale=0.32
    lODThreshold=1
    numSolverIterationsPostUpdate=2
    numSolverIterationsPreUpdate=8
    simulationSpace=RootRelative
    alphaBoolBlend {blendOption=Linear}
    boundBone {boneName=neck_02}
    componentAppliedLinearAccClamp=<100 100 100>
    componentLinearAccScale=<0.5 0.5 0.5>
    componentLinearVelScale=<0.5 0.5 0.5>
    physicsBodyDefinitions [{
      sphereCollisionRadius=0
      boundBone {boneName=neck_02}
      boxExtents=<5 5 5>
      constraintSetup {angularTargetAxis=AxisY, twistAxis=AxisY, angularLimitsMax=<15 10 2>, angularLimitsMin=<-15 -10 -20>, angularTarget=<0 0 0>}
      localJointOffset=<0 0.25 0>
     }]
    relativeSpaceBone {boneName=root}
   }
   n5 = Anim Dynamics - Bone: head(ComponentPose=@n4, Alpha=1, ExternalForce="0,0,0") {
    angularDampingOverride=0.99
    angularSpringConstant=1500
    bAngularSpring=true
    bGravityOverrideInSimSpace=true
    bOverrideAngularDamping=true
    gravityScale=0.32
    lODThreshold=1
    numSolverIterationsPostUpdate=2
    numSolverIterationsPreUpdate=8
    simulationSpace=RootRelative
    alphaBoolBlend {blendOption=Linear}
    boundBone {boneName=head}
    componentAppliedLinearAccClamp=<100 100 100>
    componentLinearAccScale=<0.5 0.5 0.5>
    componentLinearVelScale=<0.5 0.5 0.5>
    physicsBodyDefinitions [{
      sphereCollisionRadius=0
      boundBone {boneName=head}
      boxExtents=<5 5 5>
      constraintSetup {angularTargetAxis=AxisY, twistAxis=AxisY, angularLimitsMax=<15 10 2>, angularLimitsMin=<-15 -10 -20>, angularTarget=<0 0 0>}
      localJointOffset=<0 0.25 0>
     }]
    relativeSpaceBone {boneName=root}
   }
   n6 = Blend Poses by bool(
    BlendPose_0=Component To Local(ComponentPose=@n5)
    BlendPose_1=Component To Local(ComponentPose=@n3)
    BlendTime_0=1
    BlendTime_1=1
    bActiveValue=Get bIsPlayingEmoteOrDBNO() -> (bIsPlayingEmoteOrDBNO:bool)
   ) {blendType=Linear}
   n7 = Use cached pose 'Sub Graph Cache'()
   n8 = Get bPetWantsAnimDynamics() -> (bPetWantsAnimDynamics:bool)
   n9 = "Constraint - thigh_OL_l (RR)"!(
    ComponentPose=Local To Component(LocalPose=Blend Poses by bool(BlendPose_0=@n6, BlendPose_1=@n7, BlendTime_0=1, BlendTime_1=1, bActiveValue=@n8) {blendType=Linear})
    Alpha=1
    ConstraintWeights_0=1
    ConstraintWeights_1:float=1
   ) -> (Pose) {
    alphaBoolBlend {blendOption=Linear}
    boneToModify {boneName=thigh_OL_l}
    constraintSetup [{transformType=Rotation, targetBone {boneName=pelvis}}, {transformType=Rotation, targetBone {boneName=thigh_l}}]
    constraintWeights [1,1]
   }
   n10 = "Constraint - thigh_OL_r (RR)"!(ComponentPose=@n9, Alpha=1, ConstraintWeights_0=1, ConstraintWeights_1:float=1) -> (Pose) {
    alphaBoolBlend {blendOption=Linear}
    boneToModify {boneName=thigh_OL_r}
    constraintSetup [{transformType=Rotation, targetBone {boneName=pelvis}}, {transformType=Rotation, targetBone {boneName=thigh_r}}]
    constraintWeights [1,1]
   }
   n11 = "Constraint - calf_OL_l (RR)"!(ComponentPose=@n10, Alpha=1, ConstraintWeights_0=1, ConstraintWeights_1:float=1) -> (Pose) {
    alphaBoolBlend {blendOption=Linear}
    boneToModify {boneName=calf_OL_l}
    constraintSetup [{transformType=Rotation, targetBone {boneName=thigh_l}}, {transformType=Rotation, targetBone {boneName=calf_l}}]
    constraintWeights [1,1]
   }
   n12 = "Constraint - calf_OL_r (RR)"!(ComponentPose=@n11, Alpha=1, ConstraintWeights_0=1, ConstraintWeights_1:float=1) -> (Pose) {
    alphaBoolBlend {blendOption=Linear}
    boneToModify {boneName=calf_OL_r}
    constraintSetup [{transformType=Rotation, targetBone {boneName=thigh_r}}, {transformType=Rotation, targetBone {boneName=calf_r}}]
    constraintWeights [1,1]
   }
   n13 = "Constraint - heel_OL_l (RR)"!(ComponentPose=@n12, Alpha=1, ConstraintWeights_0=1, ConstraintWeights_1:float=1) -> (Pose) {
    alphaBoolBlend {blendOption=Linear}
    boneToModify {boneName=heel_OL_l}
    constraintSetup [{transformType=Rotation, targetBone {boneName=calf_l}}, {transformType=Rotation, targetBone {boneName=heel_l}}]
    constraintWeights [1,1]
   }
   n14 = "Constraint - heel_OL_r (RR)"!(ComponentPose=@n13, Alpha=1, ConstraintWeights_0=1, ConstraintWeights_1:float=1) -> (Pose) {
    alphaBoolBlend {blendOption=Linear}
    boneToModify {boneName=heel_OL_r}
    constraintSetup [{transformType=Rotation, targetBone {boneName=calf_r}}, {transformType=Rotation, targetBone {boneName=heel_r}}]
    constraintWeights [1,1]
   }
   n15 = "Constraint - foot_OL_l (RR)"!(ComponentPose=@n14, Alpha=1, ConstraintWeights_0=1, ConstraintWeights_1:float=1) -> (Pose) {
    alphaBoolBlend {blendOption=Linear}
    boneToModify {boneName=foot_OL_l}
    constraintSetup [{transformType=Rotation, targetBone {boneName=heel_l}}, {transformType=Rotation, targetBone {boneName=foot_l}}]
    constraintWeights [1,1]
   }
   n16 = "Constraint - foot_OL_r (RR)"!(ComponentPose=@n15, Alpha=1, ConstraintWeights_0=1, ConstraintWeights_1:float=1) -> (Pose) {
    alphaBoolBlend {blendOption=Linear}
    boneToModify {boneName=foot_OL_r}
    constraintSetup [{transformType=Rotation, targetBone {boneName=heel_r}}, {transformType=Rotation, targetBone {boneName=foot_r}}]
    constraintWeights [1,1]
   }
   Output Pose(Result=Component To Local(ComponentPose=@n16)) {layerGroup=None}
   AnimGraphNode_SubInput_1 = Input Pose()
   Save cached pose 'Sub Graph Cache'(Pose=@AnimGraphNode_SubInput_1) {cacheName=Sub Graph Cache}
  ]}
 EventGraph {nodes [Event BlueprintUpdateAnimation()]}
]
variables []
